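# Sound cue for the sound played when the off-road tire mod is applied to a Valet vehicle.
# It plays a single wave, Veh_Valet_Tire_Mod_Apply_01, through one wave player with no randomization or modulation.
# It routes to the Fort_Vehicles sound class and uses the shared vehicle mod items attenuation settings.
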SoundCue /ValetMods/Audio/OffRoadTire/ValetMod_OffRoadTire_Apply_Cue
root=WavePlayer_0
nodes [{id=WavePlayer_0, properties {soundWaveAssetPtr=&/Veh_Valet_Tire_Mod_Apply_01}}]
properties {
 attenuationSettings=/VehicleAudio/Settings/Vehicle_Mod_Items_Att
 maxDistance=6512
 soundClassObject=&Fort_Vehicles
}
