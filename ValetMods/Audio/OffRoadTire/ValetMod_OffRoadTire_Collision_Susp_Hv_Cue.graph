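# Sound cue for the heavy suspension-collision bounce on vehicles with the Valet off-road tire mod.
# It randomly picks one of seven equally weighted bounce recordings and passes it through a modulator for playback variation.
# It plays as an exterior vehicle sound with close-detail movement attenuation and reduced volume, and references a shared suspension-collision concurrency set alongside override settings for a single, owner-limited instance that blocks new triggers.
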SoundCue /ValetMods/Audio/OffRoadTire/ValetMod_OffRoadTire_Collision_Susp_Hv_Cue
root=Modulator_0
nodes [
 {id=Modulator_0, children [Random_1]}
 {
  id=Random_1
  children [WavePlayer_0, WavePlayer_1, WavePlayer_2, WavePlayer_10, WavePlayer_11, WavePlayer_12, WavePlayer_13]
  properties {weights [1,1,1,1,1,1,1]}
 }
 {id=WavePlayer_0, properties {soundWaveAssetPtr=&/Veh_Valet_Susp_ModBounce_Hv_01}}
 {id=WavePlayer_1, properties {soundWaveAssetPtr=&/Veh_Valet_Susp_ModBounce_Hv_02}}
 {id=WavePlayer_10, properties {soundWaveAssetPtr=&/Veh_Valet_Susp_ModBounce_Hv_04}}
 {id=WavePlayer_11, properties {soundWaveAssetPtr=&/Veh_Valet_Susp_ModBounce_Hv_05}}
 {id=WavePlayer_12, properties {soundWaveAssetPtr=&/Veh_Valet_Susp_ModBounce_Hv_06}}
 {id=WavePlayer_13, properties {soundWaveAssetPtr=&/Veh_Valet_Susp_ModBounce_Hv_07}}
 {id=WavePlayer_2, properties {soundWaveAssetPtr=&/Veh_Valet_Susp_ModBounce_Hv_03}}
]
properties {
 attenuationSettings=/VehicleAudio/Settings/Vehicle_Movement_Close_Detail_Att
 maxDistance=12000
 soundClassObject=&Fort_Vehicles_Exterior
 volumeMultiplier=0.6
 concurrencyOverrides {bLimitToOwner=true, maxCount=1, resolutionRule=PreventNew, retriggerTime=1}
 concurrencySet [/VehicleAudio/Settings/Collision_Susp_Conc]
}
