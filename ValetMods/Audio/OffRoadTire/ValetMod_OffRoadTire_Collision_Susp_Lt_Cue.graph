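# Sound cue for light suspension bounce collisions on the Valet vehicle's off-road tire mod.
# A Random node picks one of six equally weighted bounce waves and feeds a Modulator at default settings for variation.
# A Quality Level root routes that chain into its first two slots, leaving the other two empty.
# Playback uses exterior vehicle sound class, close-detail movement attenuation and shared suspension collision concurrency.
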SoundCue /ValetMods/Audio/OffRoadTire/ValetMod_OffRoadTire_Collision_Susp_Lt_Cue
root=QualityLevel_0
nodes [
 {id=Modulator_0, children [Random_1]}
 {id=QualityLevel_0, children [Modulator_0,Modulator_0,null,null]}
 {
  id=Random_1
  children [WavePlayer_0, WavePlayer_1, WavePlayer_2, WavePlayer_9, WavePlayer_10, WavePlayer_11]
  properties {weights [1,1,1,1,1,1]}
 }
 {id=WavePlayer_0, properties {soundWaveAssetPtr=&/Veh_Valet_Susp_ModBounce_Lt_03}}
 {id=WavePlayer_1, properties {soundWaveAssetPtr=&/Veh_Valet_Susp_ModBounce_Lt_04}}
 {id=WavePlayer_10, properties {soundWaveAssetPtr=&/Veh_Valet_Susp_ModBounce_Lt_01}}
 {id=WavePlayer_11, properties {soundWaveAssetPtr=&/Veh_Valet_Susp_ModBounce_Lt_02}}
 {id=WavePlayer_2, properties {soundWaveAssetPtr=&/Veh_Valet_Susp_ModBounce_Lt_05}}
 {id=WavePlayer_9, properties {soundWaveAssetPtr=&/Veh_Valet_Susp_ModBounce_Lt_06}}
]
properties {
 attenuationSettings=/VehicleAudio/Settings/Vehicle_Movement_Close_Detail_Att
 maxDistance=12000
 priority=0.8
 soundClassObject=&Fort_Vehicles_Exterior
 volumeMultiplier=0.6
 concurrencySet [/VehicleAudio/Settings/Collision_Susp_Conc]
}
