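# Sound cue for medium-intensity suspension collision bounces on Valet vehicles with the OffRoadTire mod.
# A Random node picks one of six equally weighted bounce waves and feeds a Modulator for pitch variation.
# The root Quality Level node routes its first two slots to that shared branch and leaves the other two empty.
# Close-detail vehicle attenuation, a suspension-collision concurrency set and the vehicle exterior sound class govern playback.
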SoundCue /ValetMods/Audio/OffRoadTire/ValetMod_OffRoadTire_Collision_Susp_Med_Cue
root=QualityLevel_0
nodes [
 {id=Modulator_0, children [Random_0], properties {pitchMax=1.1}}
 {id=QualityLevel_0, children [Modulator_0,Modulator_0,null,null]}
 {id=Random_0, children [WavePlayer_0,WavePlayer_1,WavePlayer_2,WavePlayer_3,WavePlayer_4,WavePlayer_5], properties {weights [1,1,1,1,1,1]}}
 {id=WavePlayer_0, properties {soundWaveAssetPtr=&/Veh_Valet_Susp_ModBounce_Med_01}}
 {id=WavePlayer_1, properties {soundWaveAssetPtr=&/Veh_Valet_Susp_ModBounce_Med_02}}
 {id=WavePlayer_2, properties {soundWaveAssetPtr=&/Veh_Valet_Susp_ModBounce_Med_03}}
 {id=WavePlayer_3, properties {soundWaveAssetPtr=&/Veh_Valet_Susp_ModBounce_Med_04}}
 {id=WavePlayer_4, properties {soundWaveAssetPtr=&/Veh_Valet_Susp_ModBounce_Med_05}}
 {id=WavePlayer_5, properties {soundWaveAssetPtr=&/Veh_Valet_Susp_ModBounce_Med_06}}
]
properties {
 attenuationSettings=/VehicleAudio/Settings/Vehicle_Movement_Close_Detail_Att
 maxDistance=12000
 priority=0.9
 soundClassObject=&Fort_Vehicles_Exterior
 volumeMultiplier=0.7
 concurrencySet [/VehicleAudio/Settings/Collision_Susp_Conc]
}
