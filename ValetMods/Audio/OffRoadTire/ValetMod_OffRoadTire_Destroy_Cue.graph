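# Sound cue for the destruction of the off-road tire vehicle mod in ValetMods.
# A Random node picks among three tire-mod destroy wave variants with equal weight.
# The cue plays at half volume in the vehicles sound class, spatialized through the gadget-placement attenuation asset with a limited max distance.
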SoundCue /ValetMods/Audio/OffRoadTire/ValetMod_OffRoadTire_Destroy_Cue
root=Random_0
nodes [
 {id=Random_0, children [WavePlayer_1,WavePlayer_2,WavePlayer_3], properties {weights [1,1,1]}}
 {id=WavePlayer_1, properties {soundWaveAssetPtr=&/Veh_Valet_Tire_Mod_Destroy_03}}
 {id=WavePlayer_2, properties {soundWaveAssetPtr=&/Veh_Valet_Tire_Mod_Destroy_01}}
 {id=WavePlayer_3, properties {soundWaveAssetPtr=&/Veh_Valet_Tire_Mod_Destroy_02}}
]
properties {
 attenuationSettings=/Game/Sounds/Attenuation/Fort_GadgetPlace_Att
 maxDistance=2000
 soundClassObject=&Fort_Vehicles
 volumeMultiplier=0.5
}
