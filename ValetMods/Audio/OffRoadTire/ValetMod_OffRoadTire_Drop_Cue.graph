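# Sound cue for the drop sound of the off-road tire mod used on Valet vehicles.
# A single wave player plays one tire-mod drop wave with no randomization or modulation.
# Output is routed to the vehicles sound class and attenuated with the shared weapon-drop falloff settings, capped by a max distance.
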SoundCue /ValetMods/Audio/OffRoadTire/ValetMod_OffRoadTire_Drop_Cue
root=WavePlayer_1
nodes [{id=WavePlayer_1, properties {soundWaveAssetPtr=&/Veh_Valet_Tire_Mod_Drop_01}}]
properties {
 attenuationSettings=&Fort_Weapon_Drop_Att
 maxDistance=3000
 soundClassObject=&Fort_Vehicles
}
