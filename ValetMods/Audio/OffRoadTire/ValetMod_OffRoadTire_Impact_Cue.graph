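# Sound cue for impacts of the valet mod off-road tire, layering rubber and washer hit sounds.
# A mixer combines one random rubber impact with a slightly louder random washer impact, each picked from two waves.
# A top modulator adds pitch variation, and the cue plays slightly lowered in pitch and volume.
# It uses small-object demolition attenuation, the impacts sound class and concurrency limits on simultaneous hits.
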SoundCue /ValetMods/Audio/OffRoadTire/ValetMod_OffRoadTire_Impact_Cue
root=Modulator_2
nodes [
 {id=Mixer_0, children [Random_31,Random_0], properties {inputVolume [1,1.25]}}
 {id=Modulator_2, children [Mixer_0], properties {pitchMax=1.02}}
 {id=Random_0, children [WavePlayer_0,WavePlayer_2], properties {weights [1,1]}}
 {id=Random_31, children [WavePlayer_92, WavePlayer_93], properties {weights [1,1]}}
 {id=WavePlayer_0, properties {soundWaveAssetPtr=/Game/Sounds/Fort_Impact_Sounds/Demolition/Washer/Washer_Impact03}}
 {id=WavePlayer_2, properties {soundWaveAssetPtr=/Game/Sounds/Fort_Impact_Sounds/Demolition/Washer/Washer_Impact08}}
 {id=WavePlayer_92, properties {soundWaveAssetPtr=/Game/Sounds/Fort_Impact_Sounds/Demolition/Rubber/Rubber_Impact02}}
 {id=WavePlayer_93, properties {soundWaveAssetPtr=/Game/Sounds/Fort_Impact_Sounds/Demolition/Rubber/Rubber_Impact03}}
]
properties {
 attenuationSettings=/Game/Sounds/Attenuation/Fort_Demolition_Hit_Small_Sized_Object_Att
 maxDistance=4000
 pitchMultiplier=0.95
 soundClassObject=&Fort_Impacts
 volumeMultiplier=0.8
 concurrencyOverrides {maxCount=3, resolutionRule=StopOldest}
 concurrencySet [&CONC_Object_Impacts_FourMax]
}
