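# Sound cue for the pickup sound of the off-road tire Valet vehicle mod.
# A single wave player feeds one pickup wave straight to the output, with no randomization or modulation.
# The cue plays at reduced volume in the vehicles sound class and borrows the gadget-throw attenuation for spatial falloff.
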SoundCue /ValetMods/Audio/OffRoadTire/ValetMod_OffRoadTire_Pickup_Cue
root=WavePlayer_0
nodes [{id=WavePlayer_0, properties {soundWaveAssetPtr=&/Veh_Valet_Tire_Mod_Pickup_01}}]
properties {
 attenuationSettings=/Game/Sounds/Attenuation/Fort_GadgetThrow_Att
 maxDistance=1536
 soundClassObject=&Fort_Vehicles
 volumeMultiplier=0.5
}
