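# Sound cue that plays the spawn sound for an off-road tire mod pile in the Valet vehicle mods.
# A single wave player feeds the tire pile wave straight to the output at reduced volume, using vehicle mod item attenuation and the vehicles sound class.
# A concurrency override keeps one instance per owner, stopping the oldest when a new one starts.
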SoundCue /ValetMods/Audio/OffRoadTire/ValetMod_OffRoadTire_Pile_Spawn_Cue
root=WavePlayer_1
nodes [{id=WavePlayer_1, properties {soundWaveAssetPtr=&/Veh_Valet_Tire_Mod_Pile_01}}]
properties {
 attenuationSettings=/VehicleAudio/Settings/Vehicle_Mod_Items_Att
 maxDistance=6512
 soundClassObject=&Fort_Vehicles
 volumeMultiplier=0.5
 concurrencyOverrides {bLimitToOwner=true, maxCount=1, resolutionRule=StopOldest}
}
