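# Sound cue for the whoosh-by of a thrown off-road tire vehicle mod in the Valet mods audio set.
# A Random node picks one of three whoosh-by wave variations with equal weight.
# The cue plays slightly pitched up in the vehicles sound class, using gadget-throw attenuation for spatial falloff.
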SoundCue /ValetMods/Audio/OffRoadTire/ValetMod_OffRoadTire_Whoosh_By_Cue
root=Random_0
nodes [
 {id=Random_0, children [WavePlayer_1,WavePlayer_2,WavePlayer_3], properties {weights [1,1,1]}}
 {id=WavePlayer_1, properties {soundWaveAssetPtr=&/Veh_Valet_Tire_Mod_Throw_Whoosh_By_02}}
 {id=WavePlayer_2, properties {soundWaveAssetPtr=&/Veh_Valet_Tire_Mod_Throw_Whoosh_By_03}}
 {id=WavePlayer_3, properties {soundWaveAssetPtr=&/Veh_Valet_Tire_Mod_Throw_Whoosh_By_01}}
]
properties {
 attenuationSettings=/Game/Sounds/Attenuation/Fort_GadgetThrow_Att
 maxDistance=1536
 pitchMultiplier=1.1
 soundClassObject=&Fort_Vehicles
}
